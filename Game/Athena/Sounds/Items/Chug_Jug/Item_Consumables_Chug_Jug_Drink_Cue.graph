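# Sound cue for drinking the Chug Jug consumable in Athena.
# A single wave player plays one use sound at boosted volume, with no randomization or modulation.
# It routes through the Fort_Consumables sound class and uses the shared Athena consumable-use attenuation for spatial falloff.
# A concurrency override caps simultaneous instances, stopping the oldest when the limit is exceeded.
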SoundCue /Game/Athena/Sounds/Items/Chug_Jug/Item_Consumables_Chug_Jug_Drink_Cue
root=WavePlayer_0
nodes [{id=WavePlayer_0, properties {soundWaveAssetPtr=&/Item_Consumables_Chug_Jug_Use_01}}]
properties {
 attenuationSettings=&Fort_Consumable_Use_Athena_Att
 bOverrideConcurrency=true
 maxDistance=2250
 soundClassObject=&Fort_Consumables
 volumeMultiplier=1.4
 concurrencyOverrides {maxCount=4, resolutionRule=StopOldest}
}
