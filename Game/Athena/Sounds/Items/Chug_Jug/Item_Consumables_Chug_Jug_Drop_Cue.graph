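# Sound cue for the Chug Jug consumable's drop sound.
# A single wave player feeds a modulator that randomizes pitch slightly on each play while keeping volume fixed.
# Playback is reduced in volume, uses the shared weapon-drop attenuation and is assigned to the loot collect sound class.
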SoundCue /Game/Athena/Sounds/Items/Chug_Jug/Item_Consumables_Chug_Jug_Drop_Cue
root=Modulator_0
nodes [
 {id=Modulator_0, children [WavePlayer_0], properties {pitchMax=1.15, pitchMin=0.9, volumeMax=1, volumeMin=1}}
 {id=WavePlayer_0, properties {soundWaveAssetPtr=&/Item_Consumables_Chug_Jug_Drop}}
]
properties {
 attenuationSettings=&Fort_Weapon_Drop_Att
 maxDistance=3000
 soundClassObject=&Fort_Loot_Collect
 volumeMultiplier=0.4
}
