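# Sound cue for equipping the Chug Jug consumable item.
# A single equip wave plays through a modulator that randomizes pitch slightly while leaving volume constant.
# It uses weapon-reload attenuation and the loot-collect sound class, with a concurrency override allowing only one instance at a time.
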SoundCue /Game/Athena/Sounds/Items/Chug_Jug/Item_Consumables_Chug_Jug_Equip_Cue
root=Modulator_0
nodes [
 {id=Modulator_0, children [WavePlayer_0], properties {pitchMax=1.1, pitchMin=0.9, volumeMax=1, volumeMin=1}}
 {id=WavePlayer_0, properties {soundWaveAssetPtr=&/Item_Consumables_Chug_Jug_Equip_01}}
]
properties {
 attenuationSettings=&Fort_Weapon_Reload_Att
 bOverrideConcurrency=true
 maxDistance=2250
 soundClassObject=&Fort_Loot_Collect
 volumeMultiplier=1.4
 concurrencyOverrides {maxCount=1}
}
